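# Voice chat reporting modal (VoiceRepModal) that hosts a step wizard over a blurred backdrop with a loading spinner.
# BeginReportingFlow builds the VoiceReport step list, reveals the wizard and hides the spinner.
# Voice Chat Audio forwards the load state to the current reporting step.
# Complete Report advances on success or rebuilds the wizard with an error list matching the upload result; the NoRecording list is defined but unused in this graph.
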WidgetBlueprint /VoiceRep/VoiceReportWizard/WBP_VoiceReport
graphs [
 EventGraph {nodes [
   n1 = Voice Chat Audio!() -> (OutputDelegate, then, Loaded:bool=~) {is=title K2Node_CustomEvent}
   > n2 = Cast To WBP_StepWizard_Base!(execute, Object=Get WBP_StepWizard_Base() -> (WBP_StepWizard_Base:Voice Rep Modal Panel Container Object Reference)) -> (then, CastFailed, AsWBP Step Wizard Base:WBP Step Wizard Base Object Reference) {is=title K2Node_DynamicCast}
   > n3 = Cast To WBP_VoiceRep_Reporting_Step!(execute, Object=Get CurrentDataWidget!(self:WBP Step Wizard Base Object Reference=@n2) -> (CurrentDataWidget:Widget Object Reference)) -> (then, CastFailed, AsWBP Voice Rep Reporting Step:WBP Voice Rep Reporting Step Object Reference) {is=title K2Node_DynamicCast}
   > Set Parent Widget!(execute, Parent Widget:Widget Object Reference=Self-Reference(), self:WBP Voice Rep Reporting Step Object Reference=@n3) -> (then, Output_Get:Widget Object Reference)
   > Load Audio!(execute, self:WBP Voice Rep Reporting Step Object Reference=@n3, Loaded:bool=@n1.Loaded) -> (then)
   n4 = Complete Report!() -> (OutputDelegate, then, Sent:bool=~, Result:EVoiceRepUploadRecordResult Enum=Success) {is=title K2Node_CustomEvent}
   > n5 = Cast To WBP_StepWizard_Base(Object=Get WBP_StepWizard_Base())
   > Set As WBP Step Wizard Base(As WBP Step Wizard Base:WBP Step Wizard Base Object Reference=@n5) -> (Output_Get:WBP Step Wizard Base Object Reference)
   > n6 = Branch(Condition=@n4.Sent)
   > Next Step(self:WBP Step Wizard Base Object Reference=Get As WBP Step Wizard Base() -> (As WBP Step Wizard Base:WBP Step Wizard Base Object Reference))
   n7 = Switch on EVoiceRepUploadRecordResult!(execute=@n6.else, Selection:EVoiceRepUploadRecordResult Enum=@n4.Result) -> (Success:Exec, Cancelled:Exec, NoAudioToUpload:Exec, EmeraldError:Exec, FailedUpload:Exec) {is=title K2Node_SwitchEnum}
   Create Step Wizard!(
    execute=@n7.Cancelled
    self:WBP Step Wizard Base Object Reference=Get As WBP Step Wizard Base()
    List Name:String=Cancelled
    Parent Widget:Widget Object Reference=Self-Reference()
   ) -> (then)
   Create Step Wizard(execute=@n7.EmeraldError, self=Get As WBP Step Wizard Base(), List Name=EmeraldError, Parent Widget=Self-Reference())
   Create Step Wizard(execute=@n7.FailedUpload, self=Get As WBP Step Wizard Base(), List Name=Error, Parent Widget=Self-Reference())
   Create Step Wizard(execute=@n7.NoAudioToUpload, self=Get As WBP Step Wizard Base(), List Name=NoAudio, Parent Widget=Self-Reference())
   Event BeginReportingFlow() {is=title K2Node_Event}
   > n8 = Cast To WBP_StepWizard_Base(Object=Get WBP_StepWizard_Base())
   > Create Step Wizard(self=@n8, List Name=VoiceReport, Parent Widget=Self-Reference())
   > Show Step Wizard()
  ]}
 Show Step Wizard {nodes [
   Show Step Wizard() {is=title K2Node_FunctionEntry}
   > SetVisibility(self=Get WBP_StepWizard_Base(), InVisibility=Visible)
   > SetVisibility(self=Get LoadingSpinner() -> (LoadingSpinner:Image Object Reference), InVisibility=Collapsed)
  ]}
]
variables [As WBP Step Wizard Base {type=WBP_StepWizard_Base_C, default=null}]
widgets {
 info {inheritedWidgetCount=1, parentClass=/Script/FortniteUI.VoiceRepModal, rootWidgetClass=&Overlay, widgetCount=7}
 widgets [
  Overlay {widgetClassPath=&Overlay, widget=Overlay}
  BackgroundBlur {
   widgetClassPath=&BackgroundBlur
   parent=Overlay
   widget=BackgroundBlur
   properties {blurStrength=8}
   slot {class=OverlaySlot, properties {horizontalAlignment=HAlign_Fill, verticalAlignment=VAlign_Fill}}
  }
  BackgroundTint {
   widgetClassPath=&Image
   parent=Overlay
   widget=BackgroundTint
   properties {colorAndOpacity=<rgba 0.0006 0.0027 0.0222 0.6>}
   slot {class=OverlaySlot, properties {horizontalAlignment=HAlign_Fill, verticalAlignment=VAlign_Fill}}
  }
  Background {
   widgetClassPath=&Image
   bIsVariable=true
   parent=Overlay
   widget=Background
   properties {
    brush {
     bIsDynamicallyLoaded=false
     drawAs=Image
     imageType=NoImage
     mirroring=NoMirror
     resourceName=None
     resourceObject=/VoiceRep/Materials/MI_UI_RadialGradient_background
     tiling=NoTile
     imageSize=<32 32>
     margin=<ltrb 0 0 0 0>
     outlineSettings {
      bUseBrushTransparency=false
      roundingType=HalfHeightRadius
      width=0
      color {colorUseRule=UseColor_Specified, specifiedColor=<rgba 0 0 0 0>}
      cornerRadii=<0 0 0 0>
     }
     tintColor {colorUseRule=UseColor_Specified, specifiedColor=<rgba 1 1 1 1>}
     uVRegion {bIsValid=false, max=<0 0>, min=<0 0>}
    }
    colorAndOpacity=<rgba 1 1 1 0.8>
   }
   slot {class=OverlaySlot, properties {horizontalAlignment=HAlign_Fill, verticalAlignment=VAlign_Fill}}
  }
  WBP_CaptureForPostBufferUpdate {
   widgetClassPath=&WBP_CaptureForPostBufferUpdate
   bIsVariable=true
   parent=Overlay
   slot {class=OverlaySlot, properties {horizontalAlignment=HAlign_Fill, verticalAlignment=VAlign_Fill}}
  }
  LoadingSpinner {
   widgetClassPath=&Image
   bIsVariable=true
   parent=Overlay
   widget=LoadingSpinner
   properties {
    brush {
     bIsDynamicallyLoaded=false
     drawAs=Image
     imageType=NoImage
     mirroring=NoMirror
     resourceName=None
     resourceObject=/Game/UI/DiscoveryShared/Art/MI_UI_Discovery_Throbber
     tiling=NoTile
     imageSize=<192 192>
     margin=<ltrb 0 0 0 0>
     outlineSettings {
      bUseBrushTransparency=false
      roundingType=HalfHeightRadius
      width=0
      color {colorUseRule=UseColor_Specified, specifiedColor=<rgba 0 0 0 0>}
      cornerRadii=<0 0 0 0>
     }
     tintColor {colorUseRule=UseColor_Specified, specifiedColor=<rgba 1 1 1 1>}
     uVRegion {bIsValid=false, max=<0 0>, min=<0 0>}
    }
    renderTransform {angle=0, scale=<1 1>, shear=<0 0>, translation=<0 -32>}
   }
   slot {class=OverlaySlot, properties {horizontalAlignment=HAlign_Center, verticalAlignment=VAlign_Center}}
  }
  WBP_StepWizard_Base {
   widgetClassPath=/VoiceRep/VoiceReportWizard/StepWizard/WBP_StepWizard_Base
   bInherited=true
   bIsVariable=true
   parent=Overlay
   widget=WBP_StepWizard_Base
   properties {
    bAutoActivate=true
    bIsFocusable=true
    bIsModal=true
    visibility=Hidden
    stepList {
     Cancelled {
      header Text=Report a voice channel
      "stepWizard Icon "=/VoiceRep/Materials/MI_UI_Icon_Megaphone
      stepWizard List [{
        button Back=false
        button Close=true
        current Step Name Text=Unable to submit report
        step Type=Error
        step Widget Content=/VoiceRep/VoiceReportWizard/VoiceReport_Steps/WBP_VoiceRep_Reporting_Step_Cancelled.WBP_VoiceRep_Reporting_Step_Cancelled_C
       }]
     }
     EmeraldError {
      header Text=Report a voice channel
      "stepWizard Icon "=&MI_UI_Icon_Megaphone
      stepWizard List [{
        button Back=false
        button Close=true
        current Step Name Text=Unable to submit report
        step Type=Error
        step Widget Content=/VoiceRep/VoiceReportWizard/VoiceReport_Steps/WBP_VoiceRep_Reporting_Step_EmeraldError.WBP_VoiceRep_Reporting_Step_EmeraldError_C
       }]
     }
     Error {
      header Text=Report a voice channel
      "stepWizard Icon "=&MI_UI_Icon_Megaphone
      stepWizard List [{
        button Back=false
        button Close=true
        current Step Name Text=Unable to submit report
        step Type=Error
        step Widget Content=/VoiceRep/VoiceReportWizard/VoiceReport_Steps/WBP_VoiceRep_Reporting_Step_Error.WBP_VoiceRep_Reporting_Step_Error_C
       }]
     }
     NoAudio {
      header Text=Report a voice channel
      "stepWizard Icon "=&MI_UI_Icon_Megaphone
      stepWizard List [{
        button Back=false
        button Close=true
        current Step Name Text=Unable to submit report
        step Type=Error
        step Widget Content=/VoiceRep/VoiceReportWizard/VoiceReport_Steps/WBP_VoiceRep_Reporting_Step_No_Audio.WBP_VoiceRep_Reporting_Step_No_Audio_C
       }]
     }
     NoRecording {
      header Text=Report a voice channel
      "stepWizard Icon "=&MI_UI_Icon_Megaphone
      stepWizard List [
       {
        button Back=false
        button Close=true
        current Step Name Text=Unable to submit report
        step Type=Error
        step Widget Content=/VoiceRep/VoiceReportWizard/VoiceReport_Steps/WBP_VoiceRep_Reporting_Step_Reporting_Off.WBP_VoiceRep_Reporting_Step_Reporting_Off_C
       }
       {
        button Back=false
        button Close=true
        current Step Name Text=Unable to submit report
        step Type=Completed
        step Widget Content=/VoiceRep/VoiceReportWizard/VoiceReport_Steps/WBP_VoiceRep_Reporting_Step_Reporting_Off_Complete.WBP_VoiceRep_Reporting_Step_Reporting_Off_Complete_C
       }
      ]
     }
     VoiceReport {
      header Text=Report a voice conversation
      "stepWizard Icon "=&MI_UI_Icon_Megaphone
      stepWizard List [
       {
        button Back=false
        button Close=true
        current Step Name Text=Submit
        step Type=Step
        step Widget Content=/VoiceRep/VoiceReportWizard/VoiceReport_Steps/WBP_VoiceRep_Reporting_Step.WBP_VoiceRep_Reporting_Step_C
       }
       {
        button Back=false
        button Close=true
        current Step Name Text=Submitted
        step Type=Completed
        step Widget Content=/VoiceRep/VoiceReportWizard/VoiceReport_Steps/WBP_VoiceRep_Reporting_Step_Complete.WBP_VoiceRep_Reporting_Step_Complete_C
       }
      ]
     }
    }
   }
   slot {class=OverlaySlot, properties {horizontalAlignment=HAlign_Fill, verticalAlignment=VAlign_Fill}}
  }
 ]
}
